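; Anim notify that spawns the Niagara effect for the firecracker emote on a Fortnite player pawn.
; In pops mode it spawns the system unattached at a socket location on the character mesh and passes a HandBone index.
; Otherwise it attaches the same system to a bone on the Head part mesh and sets burn length, red burn and spawn duration parameters.
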(fn Received_Notify
  (params (MeshComp :type Skeletal%20Mesh%20Component%20Object%20Reference) (Animation :type Anim%20Sequence%20Base%20Object%20Reference) (EventReference :type Anim%20Notify%20Event%20Reference%20Structure%20%28by%20ref%29))
  (bind _asfort_player_pawn_athena (Utilities|Casting|CastToFortPlayerPawnAthena (Components|GetOwner MeshComp)))
  (if (Variables|Default|GetFireCrackerPops)
    (bind _returnvalue (Niagara|SpawnSystematLocation (Variables|Default|GetFireCrackerEmote) (Transformation|GetSocketLocation (Variables|/Script/Engine.Character|GetMesh _asfort_player_pawn_athena) (Variables|Default|GetAttachSocketName))))
    (Effects|Components|ParticleSystem|SetIntParameter _returnvalue "HandBone" (Variables|Default|GetHandBone))
    (else
      (bind _returnvalue_2 (Niagara|SpawnSystemAttached (Variables|Default|GetFireCrackerEmote) (Customization|GetSkeletalMeshforPartType _asfort_player_pawn_athena "Head" :!class /Script/FortniteGame.FortPlayerPawn) (Variables|Default|GetAttachBoneName) "0, 0, 0" "0, 0, 0" "SnapToTarget" true))
      (Effects|Components|ParticleSystem|SetBoolParameter _returnvalue_2 "BurnLong" (Variables|Default|GetBurnLong))
      (Effects|Components|ParticleSystem|SetBoolParameter _returnvalue_2 "BurnRed" (Variables|Default|GetBurnRed))
      (Effects|Components|ParticleSystem|SetFloatParameter _returnvalue_2 "SpawnDuration" (Variables|Default|GetSpawnDuration))
      (return false))))
