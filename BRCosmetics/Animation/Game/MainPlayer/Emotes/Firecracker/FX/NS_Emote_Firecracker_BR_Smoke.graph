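# Niagara system for the Battle Royale Firecracker emote's fire, smoke and sparks, attached to the character's skeletal mesh.
# An opening burst of sparks, flickering light glow and age-faded fire accompanies a ribbon smoke trail.
# Delayed looping smoke puffs and embers stop spawning while skydiving or driving, with puffs also fading by owner speed.
# User.BurnRed switches burst colors, while User.bisHeadHidden stops the ribbon and sets sprite visibility tags.
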NiagaraSystem /BRCosmetics/Animation/Game/MainPlayer/Emotes/Firecracker/FX/NS_Emote_Firecracker_BR_Smoke
dependencies {
 usedDataInterfaces [&NiagaraDataInterfaceCurve, &NiagaraDataInterfaceColorCurve, &NiagaraDataInterfaceArrayFloat]
 usedDynamicInputs []
 usedModules [
  &SystemState
  &self:SystemSpawnScript.SystemScriptSource.SystemScriptGraph.NiagaraNodeAssignment_0.SetVariables_#1
 ]
}
emitters {
 Fire {
  emitterUpdateScript [
   &EmitterState()
   &SpawnRate(SpawnRate=15, Use Spawn Probability=false)
  ]
  particleSpawnScript [
   /SetVariables(
    Particles.ColorEnd:LinearColor=&RandomRangeLinearColor(Minimum=<rgba 0.1562 0.1445 0.1328 4>, Maximum=<rgba 0.0625 0.0578 0.0531 4>)
    Particles.ColorRed:LinearColor=&RandomRangeLinearColor(Minimum=<rgba 0.9531 0.1305 0 1>, Maximum=<rgba 0.9531 0.0099 0.0099 1>)
    Particles.ColorStart:LinearColor=&RandomRangeLinearColor(Minimum=<rgba 0.9531 0.4673 0 1>, Maximum=<rgba 0.9531 0.1823 0.001 1>)
   ) {isSetParameters=true, module=SetVariables_#2}
   /SetVariables(Particles.InitSpriteRotRate=&RandomRangeFloat(Minimum=-5, Maximum=5)) {isSetParameters=true, module=SetVariables_#3}
   /SetVariables(
    Particles.VisibilityFire:NiagaraInt32=&MakeCustomIntFromBool(Bool=$User.BurnRed, True INT=1, False INT=0)
    Particles.VisibilityRed:NiagaraInt32=&MakeCustomIntFromBool(Bool=$User.BurnRed, True INT=0, False INT=1)
   ) {isSetParameters=true}
   &InitializeParticle(
    Color=$Particles.ColorEnd
    Lifetime Max=1.6
    Lifetime Min=0.5
    Lifetime Mode=::NewEnumerator1
    Sprite Size Mode=::NewEnumerator1
    Uniform Sprite Size Max=50
    Uniform Sprite Size Min=30
    UsePositionOffset=false
    Write Material Random=true
   )
   &SkeletalMeshLocation(
    Bone Sampling Mode=::NewEnumerator4
    Interpolate Mesh Sampling=false
    MaskSpawnGroup=false
    Sampled Position Offset=<12 0 0>
    Skeletal Mesh=$User.SkeletalMesh
   )
   &SphereLocation(
    MaskSpawnGroup=false
    Override Seed=false
    Sphere Radius=5
    Surface Only=false
    UseNonUniformScale=false
    UseOffset=false
    UseSphereOrientation=false
   )
   &AddVelocityFromPoint(
    Module.Use Falloff Distance=false
    Origin Offset Coordinate Space=::Simulation
    Velocity Origin:NiagaraPosition=$Output.SkeletalMeshLocation.SampledPositionWithoutOffset
    Velocity Strength=&RandomRangeFloat(Minimum=30, Maximum=100)
   )
   /NiagaraFN/Sprite/FN_RandomizeSpriteUV()
  ]
  particleUpdateScript [
   &ParticleState()
   &Drag(Drag=3, Use Linear Drag=true, Use Rotational Drag=false)
   &DynamicMaterialParameters(
    Index 0 Param 1=&Add_Float(
     A=&RandomRangeFloat(Minimum=0, Maximum=1)
     B=&FloatFromCurve(
      FloatCurve=DataInterface() {propertyValues {Curve {keys [
          {arriveTangent=2.7533, interpMode=RCIM_Cubic, leaveTangent=2.847, tangentMode=RCTM_Break, time=0, value=0}
          {arriveTangent=0.0529, interpMode=RCIM_Cubic, leaveTangent=0.2021, tangentMode=RCTM_User, time=1, value=1}
         ]}}}
      Scale Curve=0.3
     )
    )
    Index 0 Param 3=&FloatFromCurve(
     FloatCurve=DataInterface() {propertyValues {Curve {keys [
         {arriveTangent=-0.0277, interpMode=RCIM_Cubic, leaveTangent=-0.0277, tangentMode=RCTM_User, time=0, value=0.25}
         {arriveTangent=-0.3135, interpMode=RCIM_Cubic, leaveTangent=0.2723, tangentMode=RCTM_Break, time=0.3, value=0}
        ]}}}
     Scale Curve=&FloatFromCurve(
      FloatCurve=DataInterface() {propertyValues {Curve {keys [<at 0 1 RCIM_Cubic>, {interpMode=RCIM_Cubic, tangentMode=RCTM_Break, time=0.5, value=0}]}}}
      CurveIndex=$Emitter.Age
     )
    )
    Index 0 Param 4=&FloatFromCurve(FloatCurve=DataInterface() {propertyValues {Curve {keys [<at 0 1>,<at 1 0>]}}})
    Index 1 Param 1=&FloatFromCurve(FloatCurve=DataInterface() {propertyValues {Curve {keys [<at 0 1>, {arriveTangent=4.2257, interpMode=RCIM_Cubic, leaveTangent=4.2257, tangentMode=RCTM_User, time=1, value=1}]}}})
    Index 1 Param 2=&FloatFromCurve(FloatCurve=DataInterface() {propertyValues {Curve {keys [<at 0 0>]}}})
    Index 1 Param 3=&FloatFromCurve(
     FloatCurve=DataInterface() {propertyValues {Curve {keys [<at 0 0.05 RCIM_Cubic>, {arriveTangent=0.0641, interpMode=RCIM_Cubic, leaveTangent=6.3917, tangentMode=RCTM_Break, time=0.5, value=1}]}}}
     Scale Curve=4
    )
    Index 1 Param 4=&FloatFromCurve(FloatCurve=DataInterface() {propertyValues {Curve {keys [{arriveTangent=-0.136, leaveTangent=-0.136, time=0, value=1}]}}}, Scale Curve=0.3)
    Index 2 Param 1=&FloatFromCurve(FloatCurve=DataInterface() {propertyValues {Curve {keys [
         {arriveTangent=1.2401, interpMode=RCIM_Cubic, leaveTangent=0.0353, tangentMode=RCTM_User, time=0, value=0}
         {arriveTangent=0.0618, interpMode=RCIM_Cubic, leaveTangent=0.0618, tangentMode=RCTM_User, time=1, value=1}
        ]}}})
    Index 2 Param 2=&FloatFromCurve(
     FloatCurve=DataInterface() {propertyValues {Curve {keys [
         {arriveTangent=1.9056, time=0, value=0.2}
         {arriveTangent=0.0259, interpMode=RCIM_Cubic, leaveTangent=0.0259, tangentMode=RCTM_User, time=1, value=1.4}
        ]}}}
     Scale Curve=&RandomRangeFloat(Minimum=0.5, Maximum=1)
    )
    Index 2 Param 3=&FloatFromCurve(FloatCurve=DataInterface() {propertyValues {Curve {keys [
         {interpMode=RCIM_Cubic, leaveTangent=2.3064, tangentMode=RCTM_Break, time=0, value=1}
         {interpMode=RCIM_Cubic, leaveTangent=11.1233, tangentMode=RCTM_Break, time=1, value=2}
        ]}}})
    Index 2 Param 4=&FloatFromCurve(
     FloatCurve=DataInterface() {propertyValues {Curve {keys [
         {arriveTangent=-1.7751, interpMode=RCIM_Cubic, leaveTangent=-1.7751, tangentMode=RCTM_User, time=0, value=1}
         {interpMode=RCIM_Cubic, leaveTangent=-1.2956, tangentMode=RCTM_Break, time=1, value=0.2}
        ]}}}
     Scale Curve=1.3
    )
    Module.Param0WriteEnabled=true
    Module.Param0WriteEnabled1=true
    Module.Param0WriteEnabled2=true
    Module.Param1WriteEnabled=false
    Module.Param1WriteEnabled1=true
    Module.Param1WriteEnabled2=false
    Module.Param2WriteEnabled=true
    Module.Param2WriteEnabled1=true
    Module.Param2WriteEnabled2=true
    Module.Param3WriteEnabled=false
    Module.Param3WriteEnabled1=true
    Module.Param3WriteEnabled2=true
    Write Parameter Index 1=true
    Write Parameter Index 2=true
   )
   &SolveForcesAndVelocity(Clamp Velocity=false, Limit Acceleration=false)
   &ScaleColor(
    Scale Alpha=&FloatFromCurve(
     FloatCurve=DataInterface() {propertyValues {Curve {keys [
         {arriveTangent=-0.0075, interpMode=RCIM_Cubic, leaveTangent=-0.0075, tangentMode=RCTM_User, time=0, value=0}
         {interpMode=RCIM_Cubic, tangentMode=RCTM_User, time=0.3, value=1}
         <at 1 0>
        ]}}}
     Scale Curve=&RandomRangeFloat(Minimum=0.1, Maximum=0.5)
    )
    ScaleA=true
    ScaleRGB=true
   )
   &ScaleColor(
    Scale Alpha=&FloatFromCurve(
     FloatCurve=DataInterface() {propertyValues {Curve {keys [<at 1.5 1 RCIM_Cubic>, {interpMode=RCIM_Cubic, tangentMode=RCTM_Break, time=3, value=0}]}}}
     CurveIndex=$Emitter.Age
    )
    ScaleA=true
    ScaleRGB=false
   ) {module=ScaleColor001}
   &ScaleSpriteSize(Scale Sprite Size Mode=::NewEnumerator0, Uniform Scale Factor=1.75) {module=ScaleSpriteSize001}
   &ScaleSpriteSize(Scale Factor=&Vector2DFromFloat(Value=&FloatFromCurve(
      FloatCurve=DataInterface() {propertyValues {Curve {keys [
          {arriveTangent=0.0112, interpMode=RCIM_Cubic, leaveTangent=0.0112, tangentMode=RCTM_User, time=0, value=0.7}
          {arriveTangent=-0.0415, interpMode=RCIM_Cubic, leaveTangent=-0.0415, tangentMode=RCTM_User, time=0.7, value=1}
         ]}}}
      Scale Curve=&FloatFromCurve(
       FloatCurve=DataInterface() {propertyValues {Curve {keys [
           {arriveTangent=-1.2769, interpMode=RCIM_Cubic, leaveTangent=-1.2769, tangentMode=RCTM_User, time=0.15, value=1.3}
           {interpMode=RCIM_Cubic, tangentMode=RCTM_Break, time=1, value=1}
          ]}}}
       CurveIndex=$Emitter.Age
      )
     )))
   &SpriteRotationRate(Rotation Rate=&Multiply_Float(
     A=$Particles.InitSpriteRotRate
     B=&FloatFromCurve(
      FloatCurve=DataInterface() {propertyValues {Curve {keys [
          {arriveTangent=0.1012, interpMode=RCIM_Cubic, time=0.2, value=1}
          {arriveTangent=-0.2496, interpMode=RCIM_Cubic, leaveTangent=-0.2496, tangentMode=RCTM_Break, time=1, value=0.15}
         ]}}}
      Scale Curve=&RandomRangeFloat(Minimum=2, Maximum=20)
     )
    ))
   /SetVariables(Particles.VisibilityTag=&MakeCustomIntFromBool(Bool=$User.bisHeadHidden, True INT=1, False INT=0)) {isSetParameters=true, module=SetVariables_#4}
  ]
  rendererClasses [&NiagaraSpriteRendererProperties]
  renderers [
   {rendererClass=&NiagaraSpriteRendererProperties}
   {rendererClass=&NiagaraSpriteRendererProperties, rendererIndex=1}
   {rendererClass=&NiagaraSpriteRendererProperties, rendererIndex=2}
   {rendererClass=&NiagaraSpriteRendererProperties, rendererIndex=3}
  ]
 }
 Fire001 {
  emitterUpdateScript [
   &EmitterState(Delay First Loop Only=true, Life Cycle Mode=::NewEnumerator1, Loop Delay=3, Loop Duration=1, UseLoopDelay=true)
   &SpawnRate(
    Spawn Probability=&MakeCustomFloatFromBool(Bool=&Boolean_OrOperation(Bool1=$User.IsSkydiving, Bool2=$User.IsDriving), True Float=0, False Float=1)
    SpawnRate=3
    Use Spawn Probability=true
   )
  ]
  particleSpawnScript [
   /SetVariables(
    Particles.ColorEnd:LinearColor=&RandomRangeLinearColor(Minimum=<rgba 0.4271 0.3951 0.363 4>, Maximum=<rgba 0.1615 0.1493 0.1372 4>)
    Particles.ColorRed:LinearColor=&RandomRangeLinearColor(Minimum=<rgba 0.9531 0.1305 0 1>, Maximum=<rgba 0.9531 0.0099 0.0099 1>)
    Particles.ColorStart:LinearColor=&RandomRangeLinearColor(Minimum=<rgba 0.9531 0.4673 0 1>, Maximum=<rgba 0.9531 0.1823 0.001 1>)
   ) {isSetParameters=true, module=SetVariables_#2}
   /SetVariables(Particles.InitSpriteRotRate=&RandomRangeFloat(Minimum=-5, Maximum=5)) {isSetParameters=true, module=SetVariables_#3}
   &InitializeParticle(
    Color=$Particles.ColorEnd
    Lifetime Max=1.5
    Lifetime Mode=::NewEnumerator1
    Sprite Size Mode=::NewEnumerator1
    Uniform Sprite Size Max=50
    Uniform Sprite Size Min=30
    UsePositionOffset=false
    Write Material Random=true
   )
   &SkeletalMeshLocation(Bone Sampling Mode=::NewEnumerator4, MaskSpawnGroup=false, Skeletal Mesh=$User.SkeletalMesh)
   &SphereLocation(
    MaskSpawnGroup=false
    Override Seed=false
    Sphere Radius=5
    Surface Only=false
    UseNonUniformScale=false
    UseOffset=false
    UseSphereOrientation=false
   )
   &AddVelocityFromPoint(
    Module.Use Falloff Distance=false
    Origin Offset Coordinate Space=::Simulation
    Velocity Origin:NiagaraPosition=$Output.SkeletalMeshLocation.SampledPositionWithoutOffset
    Velocity Strength=&RandomRangeFloat(Minimum=15, Maximum=35)
   )
   &FN_RandomizeSpriteUV()
   &AddVelocity(Coordinate Space=::Simulation) {enabled=false}
   /SetVariables(Particles.VelLength:NiagaraFloat=&self:Velocity_Length(Velocity:Vector3f=$Engine.Owner.Velocity)) {isSetParameters=true}
  ]
  particleUpdateScript [
   &ParticleState(Lifetime:NiagaraFloat=&Multiply_Float(
     A=$Particles.Lifetime
     B=&FloatFromCurve(
      FloatCurve=DataInterface() {propertyValues {Curve {keys [
          <at 0 1 RCIM_Cubic>
          {interpMode=RCIM_Cubic, leaveTangent=-0.0439, tangentMode=RCTM_Break, time=0.2, value=1}
          <at 0.4 0>
          {arriveTangent=0.0153, interpMode=RCIM_Cubic, leaveTangent=0.0153, tangentMode=RCTM_User, time=1, value=0}
         ]}}}
      CurveIndex=&NormalizeFloat(Value=$Particles.VelLength, Min=0, Max=800)
     )
    ))
   &Drag(Use Linear Drag=true, Use Rotational Drag=false)
   &AccelerationForce(Acceleration=&RandomRangeVector(Minimum=<0 0 10>, Maximum=<1 1 25>))
   &DynamicMaterialParameters(
    Index 0 Param 1=&Add_Float(
     A=&RandomRangeFloat(Minimum=0, Maximum=1)
     B=&FloatFromCurve(
      FloatCurve=DataInterface() {propertyValues {Curve {keys [
          {arriveTangent=2.7533, interpMode=RCIM_Cubic, leaveTangent=2.847, tangentMode=RCTM_Break, time=0, value=0}
          {arriveTangent=0.0529, interpMode=RCIM_Cubic, leaveTangent=0.2021, tangentMode=RCTM_User, time=1, value=1}
         ]}}}
      Scale Curve=0.3
     )
    )
    Index 0 Param 3=&FloatFromCurve(FloatCurve=DataInterface() {propertyValues {Curve {keys [
         {arriveTangent=-0.0277, time=0, value=0}
         {arriveTangent=-0.3135, interpMode=RCIM_Cubic, leaveTangent=0.2723, tangentMode=RCTM_Break, time=0.3, value=0}
        ]}}})
    Index 0 Param 4=&FloatFromCurve(FloatCurve=DataInterface() {propertyValues {Curve {keys [<at 0 1>,<at 1 0>]}}})
    Index 1 Param 1=&FloatFromCurve(FloatCurve=DataInterface() {propertyValues {Curve {keys [<at 0 1>, {arriveTangent=4.2257, interpMode=RCIM_Cubic, leaveTangent=4.2257, tangentMode=RCTM_User, time=1, value=1}]}}})
    Index 1 Param 2=&FloatFromCurve(FloatCurve=DataInterface() {propertyValues {Curve {keys [<at 0 0>]}}})
    Index 1 Param 3=&FloatFromCurve(
     FloatCurve=DataInterface() {propertyValues {Curve {keys [<at 0 0.05 RCIM_Cubic>, {arriveTangent=0.0641, interpMode=RCIM_Cubic, leaveTangent=6.3917, tangentMode=RCTM_Break, time=0.5, value=1}]}}}
     Scale Curve=4
    )
    Index 1 Param 4=&FloatFromCurve(FloatCurve=DataInterface() {propertyValues {Curve {keys [{arriveTangent=-0.136, leaveTangent=-0.136, time=0, value=1}]}}}, Scale Curve=0.3)
    Index 2 Param 1=&FloatFromCurve(FloatCurve=DataInterface() {propertyValues {Curve {keys [
         {arriveTangent=1.2401, interpMode=RCIM_Cubic, leaveTangent=0.0353, tangentMode=RCTM_User, time=0, value=0}
         {arriveTangent=0.0618, interpMode=RCIM_Cubic, leaveTangent=0.0618, tangentMode=RCTM_User, time=1, value=1}
        ]}}})
    Index 2 Param 2=&FloatFromCurve(
     FloatCurve=DataInterface() {propertyValues {Curve {keys [
         {arriveTangent=1.9056, time=0, value=0.2}
         {arriveTangent=0.0259, interpMode=RCIM_Cubic, leaveTangent=0.0259, tangentMode=RCTM_User, time=1, value=1.4}
        ]}}}
     Scale Curve=&RandomRangeFloat(Minimum=0.5, Maximum=1)
    )
    Index 2 Param 3=&FloatFromCurve(FloatCurve=DataInterface() {propertyValues {Curve {keys [
         {interpMode=RCIM_Cubic, leaveTangent=2.3064, tangentMode=RCTM_Break, time=0, value=1}
         {interpMode=RCIM_Cubic, leaveTangent=11.1233, tangentMode=RCTM_Break, time=1, value=2}
        ]}}})
    Index 2 Param 4=&FloatFromCurve(
     FloatCurve=DataInterface() {propertyValues {Curve {keys [
         {arriveTangent=-1.7751, interpMode=RCIM_Cubic, leaveTangent=-1.7751, tangentMode=RCTM_User, time=0, value=1}
         {interpMode=RCIM_Cubic, leaveTangent=-1.2956, tangentMode=RCTM_Break, time=1, value=0.2}
        ]}}}
     Scale Curve=1.3
    )
    Module.Param0WriteEnabled=true
    Module.Param0WriteEnabled1=true
    Module.Param0WriteEnabled2=true
    Module.Param1WriteEnabled=false
    Module.Param1WriteEnabled1=true
    Module.Param1WriteEnabled2=false
    Module.Param2WriteEnabled=true
    Module.Param2WriteEnabled1=true
    Module.Param2WriteEnabled2=true
    Module.Param3WriteEnabled=false
    Module.Param3WriteEnabled1=true
    Module.Param3WriteEnabled2=true
    Write Parameter Index 1=true
    Write Parameter Index 2=true
   )
   &SolveForcesAndVelocity(Clamp Velocity=false, Limit Acceleration=false)
   &ScaleColor(
    Scale Alpha=&FloatFromCurve(
     FloatCurve=DataInterface() {propertyValues {Curve {keys [
         {arriveTangent=-0.0075, interpMode=RCIM_Cubic, leaveTangent=-0.0075, tangentMode=RCTM_User, time=0, value=0}
         {interpMode=RCIM_Cubic, tangentMode=RCTM_User, time=0.3, value=1}
         <at 1 0>
        ]}}}
     Scale Curve=&RandomRangeFloat(Minimum=0.05, Maximum=0.1)
    )
    ScaleA=true
    ScaleRGB=true
   )
   &ScaleSpriteSize(Scale Factor=&Vector2DFromFloat(Value=&FloatFromCurve(
      FloatCurve=DataInterface() {propertyValues {Curve {keys [
          {arriveTangent=0.0112, interpMode=RCIM_Cubic, leaveTangent=0.0112, tangentMode=RCTM_User, time=0, value=0.7}
          {arriveTangent=0.7188, interpMode=RCIM_Cubic, leaveTangent=0.7188, tangentMode=RCTM_User, time=1, value=1}
         ]}}}
      Scale Curve=1.6
     )))
   &SpriteRotationRate(Rotation Rate=&Multiply_Float(
     A=$Particles.InitSpriteRotRate
     B=&FloatFromCurve(
      FloatCurve=DataInterface() {propertyValues {Curve {keys [
          {arriveTangent=0.1012, interpMode=RCIM_Cubic, time=0.2, value=1}
          {arriveTangent=-0.2496, interpMode=RCIM_Cubic, leaveTangent=-0.2496, tangentMode=RCTM_Break, time=1, value=0.15}
         ]}}}
      Scale Curve=&RandomRangeFloat(Minimum=1, Maximum=10)
     )
    ))
   &ScaleColor(
    Scale Alpha=&FloatFromCurve(
     FloatCurve=DataInterface() {propertyValues {Curve {keys [
         <at 0 1 RCIM_Cubic>
         {interpMode=RCIM_Cubic, leaveTangent=-0.0439, tangentMode=RCTM_Break, time=0.2, value=1}
         <at 0.4 0>
         {arriveTangent=0.0153, interpMode=RCIM_Cubic, leaveTangent=0.0153, tangentMode=RCTM_User, time=1, value=0}
        ]}}}
     CurveIndex=&NormalizeFloat(Value=$Particles.VelLength, Min=0, Max=800)
    )
    ScaleA=true
    ScaleRGB=false
   ) {module=ScaleColor001}
   /SetVariables(Particles.VisibilityTag=&MakeCustomIntFromBool(Bool=$User.bisHeadHidden, True INT=1, False INT=0)) {isSetParameters=true, module=SetVariables_#4}
  ]
  rendererClasses [&NiagaraSpriteRendererProperties]
  renderers [
   {rendererClass=&NiagaraSpriteRendererProperties}
   {rendererClass=&NiagaraSpriteRendererProperties, rendererIndex=1}
  ]
 }
 Glow {
  emitterUpdateScript [
   &EmitterState(Life Cycle Mode=::NewEnumerator1, Loop Behavior=::NewEnumerator1, Loop Duration=2.5, UseLoopDelay=false)
   /SetVariables() {isSetParameters=true}
   &SpawnRate(SpawnRate=10, Use Spawn Probability=false)
  ]
  particleSpawnScript [
   /SetVariables(Particles.ColorFire:LinearColor=<rgba 6 1.2914 0.5729 1>, Particles.ColorRed:LinearColor=<rgba 6 0.4343 0.3747 1>) {isSetParameters=true}
   &InitializeParticle(
    Color=&LerpLinearColors(
     StartColor=$Particles.ColorFire
     EndColor=$Particles.ColorRed
     LerpFactor=&MakeCustomFloatFromBool(Bool=$User.BurnRed, True Float=1, False Float=0)
    )
    Lifetime Max=0.6
    Lifetime Min=0.2
    Lifetime Mode=::NewEnumerator1
    Sprite Rotation Mode=::NewEnumerator2
    Sprite Size Mode=::NewEnumerator1
    Uniform Sprite Size Max=150
    Uniform Sprite Size Min=85
    UsePositionOffset=false
    Write Material Random=true
   )
   &DynamicMaterialParameters(
    Index 0 Param 1=1
    Index 1 Param 1=30
    Module.Param0WriteEnabled=true
    Module.Param0WriteEnabled1=true
    Module.Param1WriteEnabled=false
    Module.Param1WriteEnabled1=false
    Module.Param2WriteEnabled=false
    Module.Param2WriteEnabled1=false
    Module.Param3WriteEnabled=false
    Module.Param3WriteEnabled1=false
    Write Parameter Index 1=true
   )
  ]
  particleUpdateScript [
   &SkeletalMeshLocation(Bone Sampling Mode=::NewEnumerator4, MaskSpawnGroup=false, Sampled Position Offset=<15 0 0>, Skeletal Mesh=$User.SkeletalMesh)
   &CameraOffset(Camera Offset Amount=5)
   &ParticleState()
   &ScaleColor(
    Scale Alpha=&FloatFromCurve(
     FloatCurve=DataInterface() {propertyValues {Curve {keys [
         {interpMode=RCIM_Cubic, leaveTangent=5.4024, tangentMode=RCTM_Break, time=0, value=0}
         {interpMode=RCIM_Cubic, tangentMode=RCTM_Break, time=0.3, value=1}
         {arriveTangent=-1.8095, interpMode=RCIM_Cubic, tangentMode=RCTM_Break, time=1, value=0}
        ]}}}
     Scale Curve=&FloatFromCurve(
      FloatCurve=DataInterface() {propertyValues {Curve {keys [<at 0 1 RCIM_Cubic>, {interpMode=RCIM_Cubic, tangentMode=RCTM_Break, time=0.9, value=0}]}}}
      CurveIndex=$Emitter.Age
     )
    )
    Scale RGB=&VectorFromFloat(Value=1)
    ScaleA=true
    ScaleRGB=true
   )
   &ScaleSpriteSize(Uniform Curve Sprite Scale=DataInterface() {propertyValues {Curve {keys [
        {arriveTangent=5.99, interpMode=RCIM_Cubic, leaveTangent=5.99, tangentMode=RCTM_User, time=0, value=0.3}
        {interpMode=RCIM_Cubic, tangentMode=RCTM_Break, time=0.2, value=1}
        {arriveTangent=-2.1972, interpMode=RCIM_Cubic, tangentMode=RCTM_Break, time=1, value=0.3}
       ]}}})
   &SpriteRotationRate(Rotation Rate=&RandomRangeFloat(Minimum=-10, Maximum=10))
   &FN_Light_Attributes(
    Color=<rgba 0.9635 0.29 0 1>
    Create Color Attribute=true
    Create Exponent Attribute=true
    Create Position Attribute=true
    Create Radius Attribute=true
    Position:NiagaraPosition=$Particles.SkeletalMeshLocation.SampledPosition
    Radius=300
    Scale Color=&Multiply_Float(
     A=5
     B=&Waveform(
      "Waveform [1]"=::NewEnumerator0
      "[1] Phase Scale"=&RandomRangeFloat(Minimum=2, Maximum=6)
      "[1] Frequency"=1
      "[1] Amplitude Min/Max"=<0.6 1>
      "[2] Phase"=$Emitter.Age
      "[3] Phase"=$Emitter.Age
     )
    )
   )
   &SpriteRotationRate(Rotation Rate=&RandomRangeFloat(Minimum=-100, Maximum=100)) {module=SpriteRotationRate001}
   /SetVariables(Particles.VisibilityTag=&MakeCustomIntFromBool(Bool=$User.bisHeadHidden, True INT=1, False INT=0)) {isSetParameters=true, module=SetVariables_#4}
  ]
  rendererClasses [&NiagaraSpriteRendererProperties, &NiagaraLightRendererProperties]
  renderers [
   {rendererClass=&NiagaraSpriteRendererProperties}
   {rendererClass=&NiagaraLightRendererProperties, rendererIndex=1}
  ]
 }
 SmokeTrail {
  emitterUpdateScript [
   &EmitterState()
   /Game/Effects/Niagara/Modules/Ribbons/FN_SpawnRibbon_v3(
    Skeleton Mesh:NiagaraDataInterfaceSkeletalMesh=$User.SkeletalMesh
    Spawn Mode:ENiagaraSkeletalSpawning=::NewEnumerator1
    SpawnRate:NiagaraFloat=&MakeCustomFloatFromBool(Bool=$User.bisHeadHidden, True Float=0, False Float=50)
   )
  ]
  particleSpawnScript [
   &InitializeRibbon(Lifetime=0.75, Ribbon Width=30, UsePositionOffset=false, Write Material Random=true)
   &SkeletalMeshLocation(
    Bone Sampling Mode=::NewEnumerator4
    Interpolate Mesh Sampling=false
    MaskSpawnGroup=false
    Sampled Position Offset=<12 0 0>
    Skeletal Mesh=$User.SkeletalMesh
   )
  ]
  particleUpdateScript [
   &ParticleState()
   &FN_RibbonMaterial(
    Color=<rgba 0.474 0.4331 0.4236 1>
    Erosion Amount=0.75
    Erosion Pan Speed U=0.2
    Erosion Power=4
    Erossion Scale U=0.5
    Macro Distortion Amount=0.2
    Macro Pan Speed=0.7
    Macro Scale U=2
    Micro Distortion Amount=0.05
    Micro Pan Speed=0.2
    Micro Scale U=0.2
    Opacity Erosion=0.1
    Opacity Pan U Speed=0.15
    Scale Alpha=0.8
    UV0 Offset:NiagaraFloat=&RandomRangeFloat(Minimum=0, Maximum=1)
   ) {module=FN_RibbonMaterial001}
  ]
  rendererClasses [&NiagaraRibbonRendererProperties]
  renderers [{rendererClass=&NiagaraRibbonRendererProperties}]
 }
 Sparkles {
  emitterUpdateScript [
   &EmitterState(Delay First Loop Only=true, Life Cycle Mode=::NewEnumerator1, Loop Delay=3, Loop Duration=1, UseLoopDelay=true)
   &SpawnRate(
    Spawn Probability=&Multiply_Float(
     A=0.1
     B=&MakeCustomFloatFromBool(Bool=&Boolean_OrOperation(Bool1=$User.IsSkydiving, Bool2=$User.IsDriving), True Float=0, False Float=1)
    )
    SpawnRate=15
    Use Spawn Probability=true
   )
  ]
  particleSpawnScript [
   &InitializeParticle(
    Color Maximum=<rgba 0.7865 0.0313 0 1>
    Color Minimum=<rgba 0.9896 0.2066 0 1>
    Color Mode=::NewEnumerator2
    Lifetime Max=0.5
    Lifetime Min=0.2
    Lifetime Mode=::NewEnumerator1
    Sprite Rotation Mode=::NewEnumerator2
    Sprite Size Mode=::NewEnumerator1
    Uniform Sprite Size Max=5
    Uniform Sprite Size Min=2
    UsePositionOffset=false
    Write Material Random=true
   )
   &SkeletalMeshLocation(
    Bone Sampling Mode=::NewEnumerator4
    Interpolate Mesh Sampling=false
    MaskSpawnGroup=false
    Sampled Position Offset=<-5 0 0>
    Skeletal Mesh=$User.SkeletalMesh
   )
   &CylinderLocation(
    Cylinder Height=1
    Cylinder Radius=10
    Lathe Profile=DataInterface() {propertyValues {Curve {keys [<at 0 0.451 RCIM_Cubic>, {arriveTangent=0.049, interpMode=RCIM_Cubic, leaveTangent=0.049, time=0.5, value=1}, <at 1 0.5 RCIM_Cubic>]}}}
    MaskSpawnGroup=false
    Module.Enable Lathe Profile=false
    Module.Override Randomness=false
    Module.Surface Only=true
    Use Endcaps In Surface Only Mode=false
   )
   &AddVelocityFromPoint(
    Module.Use Falloff Distance=false
    Origin Offset=<0 0 -20>
    Origin Offset Coordinate Space=::Simulation
    Velocity Origin:NiagaraPosition=$Output.SkeletalMeshLocation.SampledPositionWithoutOffset
    Velocity Strength=&RandomRangeFloat(Minimum=75, Maximum=250)
   )
  ]
  particleUpdateScript [
   &Drag(Drag=5, Use Linear Drag=true, Use Rotational Drag=false)
   &AccelerationForce(Acceleration=&RandomRangeVector(Minimum=<0 0 20>, Maximum=<1 1 100>))
   &CurlNoiseForce(Noise Strength=350, Pan Noise=false)
   &SolveForcesAndVelocity(Clamp Velocity=false, Limit Acceleration=false)
   &ParticleState()
   &ScaleColor(
    Scale Alpha=&Multiply_Float(
     A=&FloatFromCurve(FloatCurve=DataInterface() {propertyValues {Curve {keys [
          {interpMode=RCIM_Cubic, leaveTangent=8.6414, tangentMode=RCTM_Break, time=0, value=0}
          {interpMode=RCIM_Cubic, tangentMode=RCTM_Break, time=0.2, value=1}
          {arriveTangent=-0.0232, interpMode=RCIM_Cubic, tangentMode=RCTM_Break, time=1, value=0}
         ]}}})
     B=1
    )
    Scale RGB=&VectorFromFloat(Value=&FloatFromCurve(
      FloatCurve=DataInterface() {propertyValues {Curve {keys [
          {interpMode=RCIM_Cubic, leaveTangent=0.1747, tangentMode=RCTM_Break, time=0, value=1}
          {interpMode=RCIM_Cubic, tangentMode=RCTM_Break, time=1, value=0.2}
         ]}}}
      CurveIndex=&RandomRangeFloat(Minimum=0, Maximum=1)
      Scale Curve=20
     ))
    ScaleA=false
    ScaleRGB=true
   )
   &ScaleSpriteSize(Uniform Curve Sprite Scale=DataInterface() {propertyValues {Curve {keys [<at 0 1 RCIM_Cubic>, {arriveTangent=-1.8095, interpMode=RCIM_Cubic, tangentMode=RCTM_Break, time=1, value=0.1}]}}})
   &SpriteRotationRate(Rotation Rate=&RandomRangeFloat(Minimum=-250, Maximum=250))
   /SetVariables(Particles.VisibilityTag=&MakeCustomIntFromBool(Bool=$User.bisHeadHidden, True INT=1, False INT=0)) {isSetParameters=true, module=SetVariables_#4}
  ]
  rendererClasses [&NiagaraSpriteRendererProperties]
  renderers [{rendererClass=&NiagaraSpriteRendererProperties}]
 }
 Sparks {
  emitterUpdateScript [
   &EmitterState(Life Cycle Mode=::NewEnumerator1, Loop Behavior=::NewEnumerator1, Loop Duration=2.5, UseLoopDelay=false)
   &SpawnRate(
    SpawnRate=&Multiply_Float(
     A=&RandomRangeFloat(Minimum=125, Maximum=200)
     B=&FloatFromCurve(
      FloatCurve=DataInterface() {propertyValues {Curve {keys [<at 0.2 1 RCIM_Cubic>, {interpMode=RCIM_Cubic, tangentMode=RCTM_Break, time=1, value=0.1}]}}}
      CurveIndex=$Emitter.Age
     )
    )
    Use Spawn Probability=false
   )
  ]
  particleSpawnScript [
   /SetVariables(
    Particles.ColorSparks:LinearColor=&RandomRangeLinearColor(Minimum=<rgba 3 1.1045 0.223 1>, Maximum=<rgba 6 1.2914 0.5729 1>)
    Particles.ColorSparksRed:LinearColor=&RandomRangeLinearColor(Minimum=<rgba 0.9531 0.0545 0 1>, Maximum=<rgba 0.7917 0.0082 0.0082 1>)
   ) {isSetParameters=true}
   &InitializeParticle(
    Color=<rgba 6 1.2914 0.5729 1>
    Lifetime Max=0.5
    Lifetime Min=0.05
    Lifetime Mode=::NewEnumerator1
    Sprite Rotation Mode=::NewEnumerator2
    Sprite Size Mode=::NewEnumerator1
    Uniform Sprite Size Max=2.5
    Uniform Sprite Size Min=1.5
    UsePositionOffset=false
    Write Material Random=true
   )
   &SkeletalMeshLocation(
    Bone Sampling Mode=::NewEnumerator4
    Interpolate Mesh Sampling=false
    MaskSpawnGroup=false
    Sampled Position Offset=<10 0 0>
    Skeletal Mesh=$User.SkeletalMesh
   )
   &CylinderLocation(
    Cylinder Height=1
    Cylinder Radius=12
    MaskSpawnGroup=false
    Module.Enable Lathe Profile=false
    Module.Override Randomness=false
    Module.Surface Only=true
    Use Endcaps In Surface Only Mode=false
   )
   &AddVelocityFromPoint(
    Module.Use Falloff Distance=false
    Origin Offset=<-15 0 0>
    Origin Offset Coordinate Space=::Simulation
    Velocity Origin:NiagaraPosition=$Output.SkeletalMeshLocation.SampledPositionWithoutOffset
    Velocity Strength=&RandomRangeFloat(Minimum=50, Maximum=600)
   ) {enabled=false}
   &AddVelocityInCone(
    Cone Angle=120
    Cone Axis=<0 0 1>
    Cone Axis Coordinate Space=::Simulation
    Module.Use Velocity Falloff On Cone Axis=false
    Velocity Strength=&RandomRangeFloat(Minimum=50, Maximum=600)
   )
  ]
  particleUpdateScript [
   &Drag(Drag=10, Use Linear Drag=true, Use Rotational Drag=false)
   &GravityForce()
   &SolveForcesAndVelocity(Clamp Velocity=false, Limit Acceleration=false)
   &ParticleState()
   &ScaleColor(
    Scale RGB=&MakeVectorFromLinearColorRGB(Color=&LerpLinearColors(
      StartColor=$Particles.ColorSparks
      EndColor=$Particles.ColorSparksRed
      LerpFactor=&MakeCustomFloatFromBool(Bool=$User.BurnRed, True Float=1, False Float=0)
     ))
    ScaleA=true
    ScaleRGB=true
   ) {module=ScaleColor001}
   &ScaleColor(
    Scale Alpha=&Multiply_Float(
     A=&FloatFromCurve(FloatCurve=DataInterface() {propertyValues {Curve {keys [
          {interpMode=RCIM_Cubic, leaveTangent=8.6414, tangentMode=RCTM_Break, time=0, value=0}
          {interpMode=RCIM_Cubic, tangentMode=RCTM_Break, time=0.2, value=1}
          {arriveTangent=-0.0232, interpMode=RCIM_Cubic, tangentMode=RCTM_Break, time=1, value=0}
         ]}}})
     B=1
    )
    Scale RGB=&VectorFromFloat(Value=&FloatFromCurve(
      FloatCurve=DataInterface() {propertyValues {Curve {keys [
          {interpMode=RCIM_Cubic, leaveTangent=0.1747, tangentMode=RCTM_Break, time=0, value=1}
          {interpMode=RCIM_Cubic, tangentMode=RCTM_Break, time=1, value=0.2}
         ]}}}
      CurveIndex=&RandomRangeFloat(Minimum=0, Maximum=1)
      Scale Curve=20
     ))
    ScaleA=false
    ScaleRGB=true
   )
   &ScaleSpriteSize(Uniform Curve Sprite Scale=DataInterface() {propertyValues {Curve {keys [<at 0 1 RCIM_Cubic>, {arriveTangent=-1.8095, interpMode=RCIM_Cubic, tangentMode=RCTM_Break, time=1, value=0.1}]}}})
   &SpriteRotationRate(Rotation Rate=&RandomRangeFloat(Minimum=-250, Maximum=250))
   /SetVariables(Particles.VisibilityTag=&MakeCustomIntFromBool(Bool=$User.bisHeadHidden, True INT=1, False INT=0)) {isSetParameters=true, module=SetVariables_#4}
  ]
  rendererClasses [&NiagaraSpriteRendererProperties]
  renderers [{rendererClass=&NiagaraSpriteRendererProperties}]
 }
 Sparks001 {
  emitterUpdateScript [
   &EmitterState(Delay First Loop Only=true, Life Cycle Mode=::NewEnumerator1, Loop Delay=3, Loop Duration=1, UseLoopDelay=true)
   &SpawnRate(
    Spawn Probability=&Multiply_Float(
     A=0.1
     B=&MakeCustomFloatFromBool(Bool=&Boolean_OrOperation(Bool1=$User.IsSkydiving, Bool2=$User.IsDriving), True Float=0, False Float=1)
    )
    SpawnRate=15
    Use Spawn Probability=true
   )
  ]
  particleSpawnScript [
   &InitializeParticle(
    Color Maximum=<rgba 0.7865 0.0313 0 1>
    Color Minimum=<rgba 0.9896 0.2066 0 1>
    Color Mode=::NewEnumerator2
    Lifetime Max=0.5
    Lifetime Min=0.2
    Lifetime Mode=::NewEnumerator1
    Sprite Rotation Mode=::NewEnumerator2
    Sprite Size Mode=::NewEnumerator1
    Uniform Sprite Size Max=2.5
    Uniform Sprite Size Min=1.5
    UsePositionOffset=false
    Write Material Random=true
   )
   &SkeletalMeshLocation(
    Bone Sampling Mode=::NewEnumerator4
    Interpolate Mesh Sampling=false
    MaskSpawnGroup=false
    Sampled Position Offset=<-5 0 0>
    Skeletal Mesh=$User.SkeletalMesh
   )
   &CylinderLocation(
    Cylinder Height=1
    Cylinder Radius=10
    Lathe Profile=DataInterface() {propertyValues {Curve {keys [<at 0 0.451 RCIM_Cubic>, {arriveTangent=0.049, interpMode=RCIM_Cubic, leaveTangent=0.049, time=0.5, value=1}, <at 1 0.5 RCIM_Cubic>]}}}
    MaskSpawnGroup=false
    Module.Enable Lathe Profile=false
    Module.Override Randomness=false
    Module.Surface Only=true
    Use Endcaps In Surface Only Mode=false
   )
   &AddVelocityFromPoint(
    Module.Use Falloff Distance=false
    Origin Offset=<0 0 -20>
    Origin Offset Coordinate Space=::Simulation
    Velocity Origin:NiagaraPosition=$Output.SkeletalMeshLocation.SampledPositionWithoutOffset
    Velocity Strength=&RandomRangeFloat(Minimum=75, Maximum=250)
   )
  ]
  particleUpdateScript [
   &Drag(Drag=5, Use Linear Drag=true, Use Rotational Drag=false)
   &AccelerationForce(Acceleration=&RandomRangeVector(Minimum=<0 0 20>, Maximum=<1 1 100>))
   &CurlNoiseForce(Noise Strength=350, Pan Noise=false)
   &SolveForcesAndVelocity(Clamp Velocity=false, Limit Acceleration=false)
   &ParticleState()
   &ScaleColor(
    Scale Alpha=&Multiply_Float(
     A=&FloatFromCurve(FloatCurve=DataInterface() {propertyValues {Curve {keys [
          {interpMode=RCIM_Cubic, leaveTangent=8.6414, tangentMode=RCTM_Break, time=0, value=0}
          {interpMode=RCIM_Cubic, tangentMode=RCTM_Break, time=0.2, value=1}
          {arriveTangent=-0.0232, interpMode=RCIM_Cubic, tangentMode=RCTM_Break, time=1, value=0}
         ]}}})
     B=1
    )
    Scale RGB=&VectorFromFloat(Value=&FloatFromCurve(
      FloatCurve=DataInterface() {propertyValues {Curve {keys [
          {interpMode=RCIM_Cubic, leaveTangent=0.1747, tangentMode=RCTM_Break, time=0, value=1}
          {interpMode=RCIM_Cubic, tangentMode=RCTM_Break, time=1, value=0.2}
         ]}}}
      CurveIndex=&RandomRangeFloat(Minimum=0, Maximum=1)
      Scale Curve=20
     ))
    ScaleA=false
    ScaleRGB=true
   )
   &ScaleSpriteSize(Uniform Curve Sprite Scale=DataInterface() {propertyValues {Curve {keys [<at 0 1 RCIM_Cubic>, {arriveTangent=-1.8095, interpMode=RCIM_Cubic, tangentMode=RCTM_Break, time=1, value=0.1}]}}})
   &SpriteRotationRate(Rotation Rate=&RandomRangeFloat(Minimum=-250, Maximum=250))
   /SetVariables(Particles.VisibilityTag=&MakeCustomIntFromBool(Bool=$User.bisHeadHidden, True INT=1, False INT=0)) {isSetParameters=true, module=SetVariables_#4}
  ]
  rendererClasses [&NiagaraSpriteRendererProperties]
  renderers [{rendererClass=&NiagaraSpriteRendererProperties}]
 }
}
summary {
 systemName=NS_Emote_Firecracker_BR_Smoke
 emitters [
  {emitterName=Sparks, rendererClasses [&NiagaraSpriteRendererProperties]}
  {emitterName=Glow, rendererClasses [&NiagaraSpriteRendererProperties, &NiagaraLightRendererProperties]}
  {emitterName=SmokeTrail, rendererClasses [&NiagaraRibbonRendererProperties]}
  {emitterName=Fire, rendererClasses [&NiagaraSpriteRendererProperties]}
  {emitterName=Sparks001, rendererClasses [&NiagaraSpriteRendererProperties]}
  {emitterName=Fire001, rendererClasses [&NiagaraSpriteRendererProperties]}
  {emitterName=Sparkles, rendererClasses [&NiagaraSpriteRendererProperties]}
 ]
 userVariables [
  User.bisHeadHidden {type=NiagaraBool}
  User.BurnLong {type=NiagaraBool}
  User.BurnRed {type=NiagaraBool}
  User.ElectricSparks {type=NiagaraBool}
  User.FrontendSpawn {type=NiagaraBool}
  User.IsDriving {type=NiagaraBool}
  User.IsSkydiving {type=NiagaraBool}
  User.SkeletalMesh {
   type=NiagaraDataInterfaceSkeletalMesh
   defaultValue {
    struct=NiagaraExt_VariableValue_DataInterface
    value {dataInterface=&self:NiagaraDataInterfaceSkeletalMesh_0}
   }
  }
 ]
}
